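# Sound cue for the plastic bottle jump foley of a biped cosmetic companion (Mimosa), routed to the pets sound class.
# A quality-level switch feeds one of two random pools: six jump waves for the first two levels, the first four for the other two.
# A modulator slightly varies pitch and volume on top of a raised base pitch and reduced volume.
# It uses companion foley attenuation and shares the biped footstep concurrency.
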SoundCue /CosmeticCompanions/Assets/Audio_Shared/Foley/Biped/PlasticBottle/SC_MImosa_Biped_PlasticBottle_Jump
root=Modulator_0
nodes [
 {id=Modulator_0, children [QualityLevel_0], properties {pitchMax=1.1, pitchMin=0.9, volumeMax=1, volumeMin=0.9}}
 {id=QualityLevel_0, children [Random_0,Random_0,Random_1,Random_1]}
 {
  id=Random_0
  children [WavePlayer_6, WavePlayer_7, WavePlayer_12, WavePlayer_13, WavePlayer_14, WavePlayer_15]
  properties {weights [1,1,1,1,1,1]}
 }
 {id=Random_1, children [WavePlayer_16, WavePlayer_17, WavePlayer_18, WavePlayer_19], properties {weights [1,1,1,1]}}
 {
  id=WavePlayer_12
  properties {soundWaveAssetPtr=/CosmeticCompanions/Assets/Audio_Shared/Foley/Biped/PlasticBottle/Wavs/Mimosa_Biped_PlasticBottle_Jump_03}
 }
 {
  id=WavePlayer_13
  properties {soundWaveAssetPtr=/CosmeticCompanions/Assets/Audio_Shared/Foley/Biped/PlasticBottle/Wavs/Mimosa_Biped_PlasticBottle_Jump_04}
 }
 {
  id=WavePlayer_14
  properties {soundWaveAssetPtr=/CosmeticCompanions/Assets/Audio_Shared/Foley/Biped/PlasticBottle/Wavs/Mimosa_Biped_PlasticBottle_Jump_05}
 }
 {
  id=WavePlayer_15
  properties {soundWaveAssetPtr=/CosmeticCompanions/Assets/Audio_Shared/Foley/Biped/PlasticBottle/Wavs/Mimosa_Biped_PlasticBottle_Jump_06}
 }
 {
  id=WavePlayer_16
  properties {soundWaveAssetPtr=/CosmeticCompanions/Assets/Audio_Shared/Foley/Biped/PlasticBottle/Wavs/Mimosa_Biped_PlasticBottle_Jump_01}
 }
 {
  id=WavePlayer_17
  properties {soundWaveAssetPtr=/CosmeticCompanions/Assets/Audio_Shared/Foley/Biped/PlasticBottle/Wavs/Mimosa_Biped_PlasticBottle_Jump_02}
 }
 {
  id=WavePlayer_18
  properties {soundWaveAssetPtr=&Mimosa_Biped_PlasticBottle_Jump_03}
 }
 {
  id=WavePlayer_19
  properties {soundWaveAssetPtr=&Mimosa_Biped_PlasticBottle_Jump_04}
 }
 {
  id=WavePlayer_6
  properties {soundWaveAssetPtr=&Mimosa_Biped_PlasticBottle_Jump_01}
 }
 {
  id=WavePlayer_7
  properties {soundWaveAssetPtr=&Mimosa_Biped_PlasticBottle_Jump_02}
 }
]
properties {
 attenuationSettings=/Game/Sounds/Attenuation/AS_Companion_Foley
 maxDistance=1200
 pitchMultiplier=1.2
 priority=0
 soundClassObject=&Fort_Creatures_Pets
 volumeMultiplier=0.45
 concurrencySet [/CosmeticCompanions/Assets/Audio_Shared/Foley/Biped/CONC_Mimosa_Biped_Footstep]
}
