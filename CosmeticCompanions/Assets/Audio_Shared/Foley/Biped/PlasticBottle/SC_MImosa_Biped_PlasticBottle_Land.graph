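# Sound cue for the plastic-bottle foley heard when a Mimosa biped cosmetic companion lands.
# A Random node picks evenly among four landing wave variations, then a Modulator applies slight random pitch and volume variation.
# It plays at reduced volume in the pets sound class with companion foley attenuation, shares the biped footstep concurrency group, and has zero priority.
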SoundCue /CosmeticCompanions/Assets/Audio_Shared/Foley/Biped/PlasticBottle/SC_MImosa_Biped_PlasticBottle_Land
root=Modulator_0
nodes [
 {id=Modulator_0, children [Random_1], properties {pitchMax=1.1, pitchMin=0.9, volumeMax=1, volumeMin=0.9}}
 {id=Random_1, children [WavePlayer_0,WavePlayer_1,WavePlayer_2,WavePlayer_3], properties {weights [1,1,1,1]}}
 {
  id=WavePlayer_0
  properties {soundWaveAssetPtr=/CosmeticCompanions/Assets/Audio_Shared/Foley/Biped/PlasticBottle/Wavs/Mimosa_Biped_PlasticBottle_Land_01}
 }
 {
  id=WavePlayer_1
  properties {soundWaveAssetPtr=/CosmeticCompanions/Assets/Audio_Shared/Foley/Biped/PlasticBottle/Wavs/Mimosa_Biped_PlasticBottle_Land_02}
 }
 {
  id=WavePlayer_2
  properties {soundWaveAssetPtr=/CosmeticCompanions/Assets/Audio_Shared/Foley/Biped/PlasticBottle/Wavs/Mimosa_Biped_PlasticBottle_Land_03}
 }
 {
  id=WavePlayer_3
  properties {soundWaveAssetPtr=/CosmeticCompanions/Assets/Audio_Shared/Foley/Biped/PlasticBottle/Wavs/Mimosa_Biped_PlasticBottle_Land_04}
 }
]
properties {
 attenuationSettings=/Game/Sounds/Attenuation/AS_Companion_Foley
 maxDistance=1200
 priority=0
 soundClassObject=&Fort_Creatures_Pets
 volumeMultiplier=0.65
 concurrencySet [/CosmeticCompanions/Assets/Audio_Shared/Foley/Biped/CONC_Mimosa_Biped_Footstep]
}
