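# Sound cue for the short plastic-bottle slide foley of biped cosmetic companions (Mimosa).
# A Random node picks one of four equally weighted wave variations.
# Playback is quieter and slightly pitched up, routed to the creature pets sound class with companion foley attenuation and lowest priority.
# It shares the Mimosa biped footstep concurrency group, so it competes with other sounds in that group for voices.
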SoundCue /CosmeticCompanions/Assets/Audio_Shared/Foley/Biped/PlasticBottle/SC_MImosa_Biped_PlasticBottle_SlideShort
root=Random_0
nodes [
 {id=Random_0, children [WavePlayer_0,WavePlayer_1,WavePlayer_2,WavePlayer_3], properties {weights [1,1,1,1]}}
 {
  id=WavePlayer_0
  properties {soundWaveAssetPtr=/CosmeticCompanions/Assets/Audio_Shared/Foley/Biped/PlasticBottle/Wavs/Mimosa_Biped_PlasticBottle_SlideShort_01}
 }
 {
  id=WavePlayer_1
  properties {soundWaveAssetPtr=/CosmeticCompanions/Assets/Audio_Shared/Foley/Biped/PlasticBottle/Wavs/Mimosa_Biped_PlasticBottle_SlideShort_02}
 }
 {
  id=WavePlayer_2
  properties {soundWaveAssetPtr=/CosmeticCompanions/Assets/Audio_Shared/Foley/Biped/PlasticBottle/Wavs/Mimosa_Biped_PlasticBottle_SlideShort_03}
 }
 {
  id=WavePlayer_3
  properties {soundWaveAssetPtr=/CosmeticCompanions/Assets/Audio_Shared/Foley/Biped/PlasticBottle/Wavs/Mimosa_Biped_PlasticBottle_SlideShort_04}
 }
]
properties {
 attenuationSettings=/Game/Sounds/Attenuation/AS_Companion_Foley
 maxDistance=1200
 pitchMultiplier=1.1
 priority=0
 soundClassObject=&Fort_Creatures_Pets
 volumeMultiplier=0.35
 concurrencySet [/CosmeticCompanions/Assets/Audio_Shared/Foley/Biped/CONC_Mimosa_Biped_Footstep]
}
